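# Physics asset for the Sportscar vehicle cosmetic body, giving it a simple collision and physics shape.
# It has a single body on the root bone, made of one box that is offset slightly forward and raised above the bone origin.
# It has no constraints, so no other parts are simulated as separate or jointed bodies.
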PhysicsAsset /VehicleCosmetics/Bodies/Sportscar/Meshes/Body_Sportscar_PhysicsDM
bodies {root [{shapeType=Box, extentX=444, extentY=216, extentZ=116, shapeName=root_box, center=<33.5 0 91.5>}]}
constraints []
